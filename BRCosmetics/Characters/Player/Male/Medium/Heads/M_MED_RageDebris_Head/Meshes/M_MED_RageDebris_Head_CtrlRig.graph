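# Control Rig for the RageDebris medium male head cosmetic that adjusts the left and right side beard master bones based on body pose.
# Three spherical pose readers measure each upper arm and the head relative to spine_05, outputting weights for Modify Transforms nodes.
# Each arm drives its own side while the head drives both, rotating and flattening the beard on local Z.
# Debug drawing is left enabled on all three readers.
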ControlRigBlueprint /BRCosmetics/Characters/Player/Male/Medium/Heads/M_MED_RageDebris_Head/Meshes/M_MED_RageDebris_Head_CtrlRig
graphs [
 RigVMFunctionLibrary {links [], nodes []}
 RigVMModel {nodes [
   RigUnit_BeginExecution = &RigUnit_BeginExecution()
   n1 = &RigUnit_SphericalPoseReader(
    ActiveRegionScaleFactors=(PositiveWidth=1,NegativeWidth=1,PositiveHeight=1,NegativeHeight=1)
    ActiveRegionSize=0.1
    Debug=(bDrawDebug=True,bDraw2D=False,bDrawLocalAxes=False,DebugScale=25,DebugSegments=20,DebugThickness=0.25)
    DriverAxis=(X=1,Y=0,Z=0)
    DriverItem=(Type=Bone,Name="upperarm_l")
    FalloffRegionScaleFactors=(PositiveWidth=1,NegativeWidth=1,PositiveHeight=1,NegativeHeight=1)
    FalloffSize=0.3
    FlipHeightScaling=False
    FlipWidthScaling=False
    OptionalParentItem=(Type=Bone,Name="spine_05")
    RotationOffset=(X=53.4883,Y=0,Z=-81.0162)
    ExecutePin=@RigUnit_BeginExecution.ExecutePin
   )
   > &RigUnit_ModifyTransforms(
    ItemToModify=((Item=(Type=Bone,Name="dyn_beard_l_MSTR"),Transform=(Rotation=(X=-0.0089,Y=0.0433,Z=0.0544,W=-0.9975),Translation=(X=0,Y=0,Z=0),Scale3D=(X=1,Y=1,Z=0.2))))
    WeightMinimum=0
    Weight=@n1.OutputParam
   )
   > n2 = &RigUnit_SphericalPoseReader(
    ActiveRegionScaleFactors=(PositiveWidth=1,NegativeWidth=1,PositiveHeight=1,NegativeHeight=1)
    ActiveRegionSize=0.1
    Debug=(bDrawDebug=True,bDraw2D=False,bDrawLocalAxes=False,DebugScale=25,DebugSegments=20,DebugThickness=0.25)
    DriverAxis=(X=-1,Y=0,Z=0)
    DriverItem=(Type=Bone,Name="upperarm_r")
    FalloffRegionScaleFactors=(PositiveWidth=1,NegativeWidth=1,PositiveHeight=1,NegativeHeight=1)
    FalloffSize=0.3
    FlipHeightScaling=False
    FlipWidthScaling=False
    OptionalParentItem=(Type=Bone,Name="spine_05")
    RotationOffset=(X=134.3895,Y=0,Z=90)
   )
   > &RigUnit_ModifyTransforms(
    ItemToModify=((Item=(Type=Bone,Name="dyn_beard_r_MSTR"),Transform=(Rotation=(X=0.0075,Y=-0.115,Z=-0.0649,W=0.9912),Translation=(X=0,Y=0,Z=0),Scale3D=(X=1,Y=1,Z=0.2))))
    WeightMinimum=0
    Weight=@n2.OutputParam
   )
   > n3 = &RigUnit_SphericalPoseReader(
    ActiveRegionScaleFactors=(PositiveWidth=1,NegativeWidth=1,PositiveHeight=1,NegativeHeight=1)
    ActiveRegionSize=0.1
    Debug=(bDrawDebug=True,bDraw2D=False,bDrawLocalAxes=False,DebugScale=25,DebugSegments=20,DebugThickness=0.25)
    DriverAxis=(X=0,Y=1,Z=0)
    DriverItem=(Type=Bone,Name="head")
    FalloffRegionScaleFactors=(PositiveWidth=1,NegativeWidth=1,PositiveHeight=1,NegativeHeight=1)
    FalloffSize=0.2
    FlipHeightScaling=False
    FlipWidthScaling=False
    OptionalParentItem=(Type=Bone,Name="spine_05")
    RotationOffset=(X=90,Y=0,Z=-70)
   )
   > &RigUnit_ModifyTransforms(
    ItemToModify=((Item=(Type=Bone,Name="dyn_beard_l_MSTR"),Transform=(Rotation=(X=-0.0089,Y=0.0433,Z=0.0544,W=-0.9975),Translation=(X=0,Y=0,Z=0),Scale3D=(X=1,Y=1,Z=0.2))))
    WeightMinimum=0
    Weight=@n3.OutputParam
   )
   > &RigUnit_ModifyTransforms(
    ItemToModify=((Item=(Type=Bone,Name="dyn_beard_r_MSTR"),Transform=(Rotation=(X=0.0075,Y=-0.115,Z=-0.0649,W=0.9912),Translation=(X=0,Y=0,Z=0),Scale3D=(X=1,Y=1,Z=0.2))))
    WeightMinimum=0
    Weight=@n3.OutputParam
   )
  ]}
]
hierarchy [
 root {loc=<0 0 0>, rot=<0 0 0 1>}
 pelvis {parent=root, loc=<0 -0.7071 93.1191>, rot=<0.0127 0.707 -0.0127 -0.707>}
 spine_01 {parent=pelvis, loc=<6.7494 -0.6388 0>, rot=<0 0 0.034 -0.9994>}
 spine_02 {parent=spine_01, loc=<7.2518 -0.035 0>, rot=<0 0 0.0238 -0.9997>}
 spine_03 {parent=spine_02, loc=<6.0104 0 0>, rot=<0 0 0.0149 -0.9999>}
 spine_04 {parent=spine_03, loc=<8.1024 0 0>, rot=<0 0 -0.0132 -0.9999>}
 spine_05 {parent=spine_04, loc=<10.0146 0 0>, rot=<0 0 -0.0015 -1>}
 neck_01 {parent=spine_05, loc=<11.3138 0.1529 0>, rot=<0 0 -0.2003 -0.9797>}
 neck_02 {parent=neck_01, loc=<4.5627 0 0>, rot=<0 0 0.0504 -0.9987>}
 head {parent=neck_02, loc=<5.3914 0 0>, rot=<0 0 0.115 -0.9934>}
 faceAttach {parent=head, loc=<-4.146 -0.8545 0>, rot=<0.0029 -0.7071 -0.0029 -0.7071>}
 C_brow_mid {parent=faceAttach, loc=<0 9.1031 12.0824>, rot=<0 0 0 -1>}
 C_jaw {parent=faceAttach, loc=<-0.0134 -0.1613 -1.4291>, rot=<0 0 0 1>}
 C_lip_lower_mid {parent=C_jaw, loc=<0.0223 19.1454 2.5121>, rot=<0 0 0 -1>}
 L_lip_lower_outer {parent=C_jaw, loc=<5.2042 10.84 4.0252>, rot=<0.1311 0.0165 -0.2937 -0.9467>}
 R_lip_lower_outer {parent=C_jaw, loc=<-5.1774 10.84 4.0252>, rot=<0.1311 -0.0165 0.2937 -0.9467>}
 teeth_lower {parent=C_jaw, loc=<0.0134 11.5397 4.2244>, rot=<0 0 0 1>}
 tongue {parent=C_jaw, loc=<0.0141 1.0371 0.7388>, rot=<-0.1568 0 0 -0.9876>}
 dyn_beard_bottom_MSTR {parent=C_jaw, loc=<-0.2023 5.3498 -2.8369>, rot=<0.3373 -0.3624 -0.5919 -0.636>}
 dyn_beard_bottom {parent=dyn_beard_bottom_MSTR, loc=<0 0 0>, rot=<0 0 0 1>}
 dyn_beard_hair_01 {parent=dyn_beard_bottom, loc=<2.7879 5.0007 -1.4297>, rot=<0.3816 0.0222 -0.2391 -0.8926>}
 dyn_beard_hair_02 {parent=dyn_beard_bottom, loc=<5.404 2.591 0.5983>, rot=<0.2655 -0.1546 -0.212 -0.9277>}
 dyn_beard_hair_03 {parent=dyn_beard_bottom, loc=<5.6701 -3.5031 -0.0882>, rot=<-0.1663 -0.0784 0.1935 -0.9637>}
 dyn_beard_hair_04 {parent=dyn_beard_bottom, loc=<1.2591 -5.4732 1.6834>, rot=<-0.3749 -0.0182 0.3913 -0.8402>}
 C_lip_upper_mid {parent=faceAttach, loc=<0.0022 20.6697 2.2805>, rot=<0.2225 0.0013 -0.0005 -0.9749>}
 C_nose_bridge {parent=faceAttach, loc=<0 13.226 8.8489>, rot=<-0.5138 0 0 -0.8579>}
 L_brow_mid {parent=faceAttach, loc=<4.4815 6.3327 12.3958>, rot=<-0.0001 -0.0042 0.2553 -0.9669>}
 L_brow_outer {parent=faceAttach, loc=<9.7383 1.86 11.6622>, rot=<-0.0287 -0.04 0.4491 -0.8921>}
 L_cheek_inner {parent=faceAttach, loc=<5.4523 9.6655 6.0265>, rot=<0.1721 -0.3727 0.238 -0.8802>}
 L_eye {parent=faceAttach, loc=<8.0864 5.7427 8.7844>, rot=<-0.0853 -0.1166 0.9625 0.2295>}
 L_eye_lid_lower_mid {parent=L_eye, loc=<0 0 0>, rot=<0.0329 0 0 -0.9995>}
 L_eye_lid_upper_mid {parent=L_eye, loc=<0 0 0>, rot=<0.0523 0 0 -0.9986>}
 L_lip_corner {parent=faceAttach, loc=<11.5604 4.0399 2.7165>, rot=<0.0055 0.0065 -0.3574 -0.9339>}
 L_lip_upper_outer {parent=faceAttach, loc=<6.0749 11.3958 3.7765>, rot=<0.2117 -0.1292 -0.1014 -0.9634>}
 R_brow_mid {parent=faceAttach, loc=<-4.4815 6.3327 12.3958>, rot=<-0.0001 0.0042 -0.2553 -0.9669>}
 R_brow_outer {parent=faceAttach, loc=<-9.7383 1.86 11.6622>, rot=<-0.0287 0.04 -0.4491 -0.8921>}
 R_cheek_inner {parent=faceAttach, loc=<-5.4523 9.6655 6.0265>, rot=<0.1721 0.3727 -0.238 -0.8802>}
 R_eye {parent=faceAttach, loc=<-8.0864 5.7426 8.7843>, rot=<0.0853 -0.1166 0.9625 -0.2295>}
 R_eye_lid_lower_mid {parent=R_eye, loc=<0 0 0>, rot=<0.0329 0 0 -0.9995>}
 R_eye_lid_upper_mid {parent=R_eye, loc=<0 0 0>, rot=<0.0523 0 0 -0.9986>}
 R_lip_corner {parent=faceAttach, loc=<-11.5604 4.0399 2.7165>, rot=<0.0055 -0.0065 0.3574 -0.9339>}
 R_lip_upper_outer {parent=faceAttach, loc=<-6.0749 11.3958 3.7765>, rot=<0.2117 0.1292 0.1014 -0.9634>}
 teeth_upper {parent=faceAttach, loc=<0 19.6394 2.2996>, rot=<0 0 0 1>}
 dyn_glasses_hider {parent=head, loc=<5.0891 1.1245 0.1504>, rot=<0 0 0 1>}
 dyn_whisker_top_l {parent=head, loc=<1.312 18.577 -2.1341>, rot=<-0.0883 -0.5077 -0.0572 -0.8551>}
 dyn_whisker_mid_l {parent=head, loc=<-0.3797 18.6837 -2.9289>, rot=<-0.0556 -0.703 -0.0611 -0.7064>}
 dyn_whisker_bottom_l {parent=head, loc=<-1.2948 19.2467 -2.7264>, rot=<-0.0986 -0.8588 -0.1901 -0.4654>}
 dyn_whisker_top_r {parent=head, loc=<1.3879 18.6714 2.1137>, rot=<0.8481 0.0758 0.511 0.118>}
 dyn_whisker_top_r_FX {parent=dyn_whisker_top_r, loc=<12.97 0 0>, rot=<-0.0517 0.0082 0.1562 -0.9863>}
 dyn_whisker_mid_r {parent=head, loc=<-0.3217 18.6795 2.9294>, rot=<0.7067 0.0611 0.7026 0.0556>}
 dyn_whisker_bottom_r {parent=head, loc=<-1.2918 19.2541 2.7596>, rot=<0.4334 0.201 0.8733 0.0955>}
 dyn_hair_top {parent=head, loc=<12.3163 2.0146 0.296>, rot=<0.6724 -0.2582 0.2725 -0.6379>}
 dyn_tuft_back {parent=head, loc=<0.5325 -9.5756 -1.0421>, rot=<0.6106 -0.3869 0.4574 -0.518>}
 dyn_hair_back_01 {parent=dyn_tuft_back, loc=<2.98 -0.6012 0.6898>, rot=<0.0013 -0.06 0.006 -0.9982>}
 dyn_hair_back_02 {parent=dyn_tuft_back, loc=<0.2829 3.1353 1.2576>, rot=<-0.0075 -0.1052 -0.043 -0.9935>}
 dyn_hair_back_03 {parent=dyn_tuft_back, loc=<3.9171 -1.3079 2.0658>, rot=<-0.0079 -0.2146 -0.1321 -0.9677>}
 dyn_ear_r_MSTR {parent=head, loc=<9.2875 -3.3579 8.3203>, rot=<-0.6756 0.052 -0.2879 -0.6768>}
 dyn_ear_r {parent=dyn_ear_r_MSTR, loc=<0 0 0>, rot=<0 0 0 1>}
 dyn_ear_l_MSTR {parent=head, loc=<8.6735 -2.983 -9.3361>, rot=<-0.5162 -0.2844 0.0841 -0.8035>}
 dyn_ear_l {parent=dyn_ear_l_MSTR, loc=<0 0 0>, rot=<0 0 0 1>}
 dyn_beard_l_MSTR {parent=head, loc=<0.6058 -3.8662 -11.4703>, rot=<0.23 -0.6227 0.1936 -0.7224>}
 dyn_beard_l {parent=dyn_beard_l_MSTR, loc=<0 0 0>, rot=<0 0 0 1>}
 dyn_hair_l_01 {parent=dyn_beard_l, loc=<-0.0587 -5.7308 2.3966>, rot=<-0.0021 -0.0638 -0.0263 -0.9976>}
 dyn_hair_l_02 {parent=dyn_beard_l, loc=<0.6109 -3.702 -0.7413>, rot=<0.0152 -0.0788 0.2377 -0.968>}
 dyn_hair_l_03 {parent=dyn_beard_l, loc=<2.4135 -3.0097 0.8152>, rot=<-0.012 -0.1904 0.2314 -0.954>}
 dyn_hair_l_04 {parent=dyn_beard_l, loc=<5.1732 2.3802 1.1216>, rot=<0.1212 0.0831 0.1514 -0.9775>}
 dyn_hair_l_04_FX {parent=dyn_hair_l_04, loc=<5.3733 0 0>, rot=<-0.5481 0.4055 -0.3985 -0.6135>}
 dyn_hair_l_05 {parent=dyn_beard_l, loc=<2.783 2.618 0.4643>, rot=<0.0995 -0.0763 -0.0806 -0.9888>}
 dyn_hair_l_06 {parent=dyn_beard_l, loc=<-0.6235 0.8155 -3.4711>, rot=<0.0548 -0.4904 -0.1398 -0.8585>}
 dyn_beard_r_MSTR {parent=head, loc=<0.8583 -4.8861 11.6137>, rot=<0.679 -0.2387 0.6454 -0.256>}
 dyn_beard_r {parent=dyn_beard_r_MSTR, loc=<0 0 0>, rot=<0 0 0 1>}
 dyn_hair_r_01 {parent=dyn_beard_r, loc=<2.8878 -3.2012 1.5033>, rot=<-0.0149 0.2216 -0.0536 -0.9736>}
 dyn_hair_r_02 {parent=dyn_beard_r, loc=<2.0133 1.2423 -0.0533>, rot=<0.0009 0.0558 0.0082 -0.9984>}
 dyn_hair_r_03 {parent=dyn_beard_r, loc=<6.059 0.853 -0.4002>, rot=<0.0068 -0.1204 -0.0923 -0.9884>}
 dyn_hair_r_04 {parent=dyn_beard_r, loc=<-2.6243 -1.7872 -6.5869>, rot=<0.0075 -0.3141 -0.027 -0.949>}
 dyn_hair_r_05 {parent=dyn_beard_r, loc=<-1.4631 -5.5356 -4.197>, rot=<0.0113 -0.1519 -0.1067 -0.9826>}
 clavicle_l {parent=spine_05, loc=<4.3575 -0.1466 -2.3223>, rot=<0.0465 -0.7289 -0.009 -0.6829>}
 upperarm_l {parent=clavicle_l, loc=<15.5111 0 0>, rot=<0.0226 -0.3234 -0.0088 -0.9459>}
 upperarm_twist_01_l {parent=upperarm_l, loc=<0.0041 0 0>, rot=<0 0 0 -1>}
 lowerarm_l {parent=upperarm_l, loc=<26.6514 0 0>, rot=<0.0391 -0.0099 -0.2453 -0.9686>}
 clavicle_r {parent=spine_05, loc=<4.3575 -0.1466 2.3223>, rot=<0.7289 0.0465 -0.6829 0.009>}
 upperarm_r {parent=clavicle_r, loc=<-15.5111 0 0>, rot=<0.0226 -0.3234 -0.0088 -0.9459>}
 upperarm_twist_01_r {parent=upperarm_r, loc=<-0.0041 0 0>, rot=<0 0 0 -1>}
 lowerarm_r {parent=upperarm_r, loc=<-26.6514 0 0>, rot=<0.0391 -0.0099 -0.2453 -0.9686>}
]
